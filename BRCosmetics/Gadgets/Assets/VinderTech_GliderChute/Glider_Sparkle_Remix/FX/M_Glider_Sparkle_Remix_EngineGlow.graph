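# Translucent unlit mesh-particle material for the engine glow effect on the Sparkle Remix glider.
# Particle color feeds an emissive adjustment function, with a hue shift branch driven by the outer mask.
# Laser-sword trail noise scrolls in opposite directions on each half of V, selected by rounding V.
# That noise is multiplied by particle alpha and two V-only sphere masks, one inverted to carve out the center.
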Material /BRCosmetics/Gadgets/Assets/VinderTech_GliderChute/Glider_Sparkle_Remix/FX/M_Glider_Sparkle_Remix_EngineGlow
graph {
 properties {bUsedWithMeshParticles=true, blendMode=BLEND_Translucent, pixelDepthOffsetMode=PDOM_Legacy, shadingModel=MSM_Unlit, usedShadingModels=Unlit}
 nodes [
  n1 = SphereMask(
   A=TextureCoordinate() {uTiling=0}
   B=Constant2Vector() {g=0.5}
   Radius=ScalarParameter() {defaultValue=0.4, parameterName=Radius}
   Hardness=ScalarParameter() {parameterName=Hardness}
  )
  n2 = ParticleColor()
  MaterialFunctionCall(Hue Shift Percentage=@n1, Texture=@n2.RGB) {materialFunction=&HueShift}
  MaterialFunctionCall(Emissive=@n2.RGB) {materialFunction=&MF_FX_EmissiveAdjustments}
  n3 = Panner(Coordinate=TextureCoordinate() {uTiling=3, vTiling=0.25}) {speedY=-1}
  n4 = Panner(Coordinate=TextureCoordinate() {vTiling=0.25}) {speedY=1}
  n5 = ComponentMask(None=TextureCoordinate() {uTiling=0}) {g=true}
  n6 = OneMinus(None=SphereMask(
    A=TextureCoordinate() {uTiling=0}
    B=Constant2Vector() {g=0.5}
    Radius=ScalarParameter() {defaultValue=0.25, parameterName=Radius2}
    Hardness=ScalarParameter() {parameterName=Hardness2}
   ))
  Multiply(
   A=LinearInterpolate(
    A=TextureSample(UVs=@n3) {texture=/Game/Weapons/FORT_Melee/Effects/LaserSwordFounders/T_LaserSwordTrailNoise}.G
    B=TextureSample(UVs=@n4) {texture=&T_LaserSwordTrailNoise}.G
    Alpha=Round(None=@n5)
   )
   B=Multiply(A=@n2.A, B=Multiply(A=@n1, B=@n6))
  )
 ]
}
