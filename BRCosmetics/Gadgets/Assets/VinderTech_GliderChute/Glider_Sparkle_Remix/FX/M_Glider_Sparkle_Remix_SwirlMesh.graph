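# Translucent, unlit, two-sided mesh-particle material for the swirl effect on the Sparkle Remix glider.
# Particle color drives emissive, while opacity multiplies particle alpha, a selectable ramp channel and a smoke-trail texture with swizzled UVs and a random offset.
# Dynamic parameters supply Width, Random and DTAA, with width feeding SplineThicken and a camera-relative cross-product offset.
# A shading-path switch adds depth fade on the default path and omits it on mobile.
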Material /BRCosmetics/Gadgets/Assets/VinderTech_GliderChute/Glider_Sparkle_Remix/FX/M_Glider_Sparkle_Remix_SwirlMesh
graph {
 properties {
  bUsedWithMeshParticles=true
  blendMode=BLEND_Translucent
  pixelDepthOffsetMode=PDOM_Legacy
  shadingModel=MSM_Unlit
  twoSided=true
  usedShadingModels=Unlit
 }
 nodes [
  ComponentMask(None=TextureCoordinate()) {g=true}
  n1 = ParticleColor()
  MaterialFunctionCall(Emissive=@n1.RGB) {materialFunction=&MF_FX_EmissiveAdjustments}
  n2 = DynamicParameter() {paramNames [Width,DTAA,Random,Param4]}
  n3 = LinearInterpolate(B=ScalarParameter() {defaultValue=2, parameterName=Width}, Alpha=@n2.Width)
  MaterialFunctionCall(WidthBase=@n3, WidthTip=@n3) {materialFunction=&SplineThicken}
  n4 = CrossProduct(
   A=Transform(None=Constant3Vector() {constant=<rgba 0 0 1 0>})
   B=Transform(None=Constant3Vector() {constant=<rgba 0 0 1 0>}) {transformSourceType=TRANSFORMSOURCE_Camera}
  )
  n5 = Normalize(VectorInput=@n4)
  n6 = ComponentMask(None=TextureCoordinate()) {r=true}
  Multiply(A=Multiply(A=StaticSwitchParameter(True=@n5, False=@n4) {parameterName=FroceWidth}, B=Subtract(A=@n6) {constB=0.5}), B=@n3)
  n7 = MaterialFunctionCall(
   X=TextureSample() {texture=/Game/Effects/Fort_Effects/Textures/Ramps/T_Power_Ball_Beam_Ramp}.R
   Y=TextureSample() {texture=/Game/Effects/Fort_Effects/Textures/Ramps/T_Ramp_Smooth}.R
  ) {materialFunction=&MakeFloat2}
  n8 = Constant2Vector() {r=1}
  n9 = TextureSample(UVs=Add(A=MaterialFunctionCall() {materialFunction=&Swizzle}.YX, B=Multiply(A=@n8, B=@n2.Random))) {
   samplerSource=SSM_Wrap_WorldGroupSettings
   samplerType=SAMPLERTYPE_Grayscale
   texture=&T_Smoke_Trail_Soft
  }
  n10 = Multiply(A=@n1.A, B=Multiply(A=StaticComponentMaskParameter(None=@n7.Result) {defaultR=true, parameterName=MaskChennel}, B=@n9.R))
  Multiply(
   A=LinearInterpolate(Alpha=MaterialFunctionCall() {materialFunction=&DitherTemporalAA}.Result) {constB=3}
   B=@n10
  )
  n11 = ScalarParameter() {defaultValue=200, parameterName=DF_FD}
  n12 = LinearInterpolate(A=@n10, B=@n10, Alpha=LinearInterpolate(B=DepthFade(FadeDistance=@n11), Alpha=@n2.DTAA))
  ShadingPathSwitch(Default=DepthFade(Opacity=@n12, FadeDistance=ScalarParameter() {defaultValue=50, parameterName=OpacityDF}), Mobile=@n12) {desc=Migrated from FeatureLevelSwitch}
 ]
}
